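; Growth manager for the Figment cube island that stores growth actor placements and spawns those unlocked by the current growth stage.
; On world ready, the authority reads the island travel event and floors its cached time into GrowthStage, whose RepNotify spawns entries with steps at or below it.
; Editor events spawn stored growths for editing, then harvest them back into the Growths array as relative transform, class and step.
; In edit mode, the construction script previews levels by toggling visibility.
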(event EventOnWorldReady
  ;[1 1 1] OnWorldReady
  (Networking|SwitchHasAuthority
    (:Authority
      (bind (_asynctaskproxy _payload) (GameplayEventRouter|StartListeningToStatefulEvent (GameplayEventRouter|GetGlobalEventRouter) "/Figment_S06/Gameplay/EventRouter/Figment_CubeIslandTravel_Event.Figment_CubeIslandTravel_Event")
        (:OnEventReceived
          _)
        (:OnSavedEventStateExists
          _))
      (if (Utilities|Struct|BreakFigmentCubeIslandTravelEvent _payload)
        (Variables|Default|SetGrowthStage (Math|Float|Floor (.CachedTime_4_CA4D76274A8BF3C875DC92A057FE963B (Utilities|Struct|BreakFigmentCubeIslandTravelEvent _payload))))))))

(event Custom|SpawnGrowthBPs
  ;[1 1 1] Spawn Growth BPs
  (for _array_element (Variables|Default|GetGrowths)
    (bind (_array_element_transform_2_22efd07343d8672b145c73930d835df0 _array_element_class_9_16b0c0664962358c41079193300ef569 _array_element_step_8_55dcf7bf45d7b0ab5dc2f8a8900029fb) (Utilities|Struct|BreakGrowthData _array_element))
    (if (Utilities|ClassIsChildOf _array_element_class_9_16b0c0664962358c41079193300ef569 "/Figment_S06/Athena/Environments/Island/CubeGrowth/BP_Growth_Parent.BP_Growth_Parent_C")
      (Networking|SwitchHasAuthority
        (:Authority
          _))
      (else
        (Networking|CanExecuteCosmeticEvents
          (:True
            _))))
    (if (Math|Byte|Byte<=Byte _array_element_step_8_55dcf7bf45d7b0ab5dc2f8a8900029fb (Math|Conversions|ToByte(Integer) (Variables|Default|GetGrowthStage)))
      (Game|SpawnActorfromClass _array_element_class_9_16b0c0664962358c41079193300ef569 (Math|Transform|ComposeTransforms _array_element_transform_2_22efd07343d8672b145c73930d835df0 (Transformation|GetActorTransform)) "AlwaysSpawn" "SelectDefaultAtRuntime" 0 0 0 0))))

(event Custom|Save/HarvestGrowth
  ;[0.0916 0.2553 1] Save
  (if (Variables|Default|GetPreviewMode)
    _
    (elif (Variables|Default|GetEditMode(Don'tTouch))
      (Utilities|Array|Clear (Variables|Default|GetGrowths))
      (bind _outactors (Actor|GetAllActorsOfClass "/Figment_S06/Athena/Environments/Island/CubeGrowth/BP_Growth_Parent.BP_Growth_Parent_C"))
      (for _array_element _outactors
        (Utilities|Array|Add (Variables|Default|GetGrowths) (Utilities|Struct|MakeGrowthData (Math|Transform|MakeRelativeTransform (Transformation|GetActorTransform _array_element) (Transformation|GetActorTransform)) (Utilities|GetClass _array_element) (Variables|/Figment_S06/Athena/Environments/Island/CubeGrowth/BP_Growth.BP_Growth_C|GetGrowthStep _array_element)))
        (Actor|DestroyActor _array_element))
      (Variables|Default|SetEditMode(Don'tTouch) false))))

(event Custom|EditGrowth
  ;[0.0916 0.2553 1] Edit
  (if (Variables|Default|GetPreviewMode)
    _
    (elif (Variables|Default|GetEditMode(Don'tTouch))
      _
      (else
        (for _array_element (Variables|Default|GetGrowths)
          (bind (_array_element_transform_2_22efd07343d8672b145c73930d835df0 _array_element_class_9_16b0c0664962358c41079193300ef569 _array_element_step_8_55dcf7bf45d7b0ab5dc2f8a8900029fb) (Utilities|Struct|BreakGrowthData _array_element))
          (Game|SpawnActorfromClass _array_element_class_9_16b0c0664962358c41079193300ef569 (Math|Transform|ComposeTransforms _array_element_transform_2_22efd07343d8672b145c73930d835df0 (Transformation|GetActorTransform)) "AlwaysSpawn" "SelectDefaultAtRuntime" 0 _array_element_step_8_55dcf7bf45d7b0ab5dc2f8a8900029fb 0 0))
        (Variables|Default|SetEditMode(Don'tTouch) true)))))

(event EventBeginPlay)

(event Collision|EventActorBeginOverlap
  (params (OtherActor :type Actor%20Object%20Reference)))

(event EventTick
  (params (DeltaSeconds :type Float%20%28single-precision%29)))

(fn OnRep_GrowthStage
  (CallFunction|SpawnGrowthBPs))

(fn ConstructionScript
  (if (Variables|Default|GetEditMode(Don'tTouch))
    (bind _outactors (Actor|GetAllActorsOfClass "/Figment_S06/Athena/Environments/Island/CubeGrowth/BP_Growth_Parent.BP_Growth_Parent_C"))
    (for _array_element _outactors
      (if (Variables|Default|GetPreviewMode)
        (if (Math|Byte|Byte<=Byte (Variables|/Figment_S06/Athena/Environments/Island/CubeGrowth/BP_Growth.BP_Growth_C|GetGrowthStep _array_element) (Variables|Default|GetPreviewLevel))
          (Rendering|SetVisibility (Variables|/Script/Engine.Actor|GetRootComponent _array_element) true true)
          (else
            (Rendering|SetVisibility (Variables|/Script/Engine.Actor|GetRootComponent _array_element) false true)))
        (else
          (Rendering|SetVisibility (Variables|/Script/Engine.Actor|GetRootComponent _array_element) true true))))
    (else
      (Variables|Default|SetPreviewMode false))))
