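# Translucent deferred decal material for CubeGrowth island environment art, producing an animated dark blue-to-purple noise pattern.
# A soft SphereMask on the light vector is modulated by a sine of time times Speed, offset by the mask, creating a pulse.
# That pulse sets the HeightLerp transition phase over an inverted, high-power cube noise height texture.
# The resulting alpha blends the E1 and E2 color parameters.
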Material /Figment_S06/Athena/Environments/Island/CubeGrowth/M_CubeGrowth_Decal
graph {
 properties {blendMode=BLEND_Translucent, materialDomain=MD_DeferredDecal, pixelDepthOffsetMode=PDOM_Legacy, usedShadingModels=Default Lit}
 nodes [
  n1 = Constant() {r=0.5}
  n2 = SphereMask(A=LightVector(), B=@n1, Radius=@n1) {hardnessPercent=0}
  n3 = Time()
  n4 = ScalarParameter() {defaultValue=0.2, parameterName=Speed}
  n5 = Sine(None=Add(A=@n2, B=Multiply(A=@n3, B=@n4)))
  n6 = Add(A=@n2, B=Multiply(A=@n2, B=ConstantBiasScale(None=@n5) {scale=0.15}))
  n7 = Power(Base=TextureSample() {samplerType=SAMPLERTYPE_LinearColor, texture=/Game/Athena/Environments/Island/CubeGrowth/T_CubeNoise01_H}.R) {constExponent=8}
  LinearInterpolate(
   A=VectorParameter() {parameterName=E1, defaultValue=<rgba 0.0079 0.0029 0.0729 1>}.RGB
   B=VectorParameter() {parameterName=E2, defaultValue=<rgba 0.1927 0 0.2 1>}.RGB
   Alpha=MaterialFunctionCall(Transition Phase=Clamp(None=@n6), Height Texture=OneMinus(None=@n7)) {materialFunction=&HeightLerp}.Alpha
  )
 ]
}
